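; Spawner actor for the premium supply drop that places a drop pod in the world.
; BeginPlay calls Spawn Supply Drop, which sets HasAlreadySpawned and spawns PremiumDropPodClass at the spawner's own transform.
; If the flag is already set when the event runs, the spawner destroys itself instead.
; Tick, overlap and the construction script are empty.
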(event Custom|Spawn%20Supply%20Drop
  (if (Variables|Default|GetHasAlreadySpawned)
    (Actor|DestroyActor)
    (else
      (Variables|Default|SetHasAlreadySpawned true)
      (Game|SpawnActorfromClass (Variables|Default|GetPremiumDropPodClass) (Transformation|GetActorTransform) "Undefined" "MultiplyWithRoot" 0 0 0 0))))

(event EventBeginPlay
  (CallFunction|SpawnSupplyDrop))

(event Collision|EventActorBeginOverlap
  (params (OtherActor :type Actor%20Object%20Reference)))

(event EventTick
  (params (DeltaSeconds :type Float%20%28single-precision%29)))

(fn ConstructionScript)
