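; BGA_Screen_PremiumSupplyDrop is an interactable Fortnite building actor that acts as a screen for triggering the opening of a premium supply drop.
; On BeginPlay it binds its interact event, and its BlueprintCanInteract override always allows interaction.
; On interact it prints a "ping" debug string, calls the OpenPremiumSupplyDrop dispatcher, then destroys itself, making it single-use.
; Overlap, tick and construction logic are empty.
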(event EventBeginPlay
  (Interaction|BindEventtoOnInteract (delegate Custom|OnInteract_Event)))

(event Custom|OnInteract_Event
  (params (SelfActor :type Building%20Actor%20Object%20Reference) (InteractingPawn :type Fort%20Pawn%20Object%20Reference))
  (Development|PrintString "ping")
  (Default|CallOpenPremiumSupplyDrop)
  (Actor|DestroyActor))

(event Collision|EventActorBeginOverlap
  (params (OtherActor :type Actor%20Object%20Reference)))

(event EventTick
  (params (DeltaSeconds :type Float%20%28single-precision%29)))

(fn BlueprintCanInteract
  (params (InteractingPawn :type Fort%20Pawn%20Object%20Reference) (InteractionBeingAttempted :type EInteractionBeingAttempted%20Enum :default "FirstInteraction") (InteractionType :type TInteractionType%20Enum :default "IT_NoInteraction"))
  (return true))

(fn ConstructionScript)
