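; Mini supply drop pod that applies Tidal Crane materials and plays a sound while a player opens it.
; On BeginPlay it takes the first B_PremiumSupplyDrop_Manager found, sets its Loot Tier Name from GetUniqueLTG, and debug-prints it.
; It then removes an entry from the manager's available mini-drop loot tier list, apparently so drops get unique loot.
; Beginning interaction spawns the opening sound, which stops when the pod opens or interaction is interrupted.
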(event EventBeginPlay
  (bind _staticmeshcomponent (Variables|Debug|GetStaticMeshComponent))
  (|CallParentFunction)
  (Rendering|Material|SetMaterial _staticmeshcomponent 0 "/DropPodParentGameplay/Assets/Materials/MI_DropPod_TidalCrane.MI_DropPod_TidalCrane")
  (Rendering|Material|SetMaterial _staticmeshcomponent 1 "/DropPodParentGameplay/Assets/Materials/MI_DropPod_TidalCrane_Glass.MI_DropPod_TidalCrane_Glass")
  (bind _outactors (Actor|GetAllActorsOfClass "/PremiumSupplyDrop/Gameplay/B_PremiumSupplyDrop_Manager.B_PremiumSupplyDrop_Manager_C"))
  (bind _output_get (Variables|Default|SetSupply%20Drop%20Manager (Utilities|Array|Get(acopy) _outactors)))
  (bind _selectedltg (Class|BPremiumSupplyDropManager|GetUniqueLTG _output_get))
  (bind _output_get_2 (Variables|DropPodSettings|SetLoot%20Tier%20Name _selectedltg))
  (Development|PrintString (Utilities|String|ToString(Name) _output_get_2) true true "(R=1.000000,G=0.000000,B=0.102644,A=1.000000)" 1.9e+11)
  (Utilities|Array|RemoveIndex (Variables|/PremiumSupplyDrop/Gameplay/B_PremiumSupplyDrop_Manager.B_PremiumSupplyDrop_Manager_C|GetAvailableLTGs_MiniSupplyDrop _output_get)))

(event EventOnOpened
  ; container opening and idle audio loop
  (|CallParentFunction)
  (Audio|Components|Audio|Stop (stale Interaction|EventBlueprintOnBeginInteract _returnvalue)))

(event Interaction|EventBlueprintOnInterruptInteract
  ; container opening and idle audio loop
  (Audio|Components|Audio|Stop (stale Interaction|EventBlueprintOnBeginInteract _returnvalue)))

(event Interaction|EventBlueprintOnBeginInteract
  (params (InteractingPawn :type Fort%20Pawn%20Object%20Reference) (InteractionBeingAttempted :type EInteractionBeingAttempted%20Enum :default "FirstInteraction") (InteractionType :type TInteractionType%20Enum :default "IT_NoInteraction"))
  ; container opening and idle audio loop
  (|CallParentFunction InteractingPawn InteractionBeingAttempted InteractionType)
  (bind _returnvalue (Audio|SpawnSoundatLocation :Sound "/PremiumSupplyDrop/Assets/Audio/SC_PremiumSupplyDrop_Loot_Open.SC_PremiumSupplyDrop_Loot_Open" :Location (Transformation|GetActorLocation) :VolumeMultiplier 1e+01 :PitchMultiplier 1.0 :StartTime 0.0 :bAutoDestroy true)))

(fn ConstructionScript
  (|CallParentFunction))
