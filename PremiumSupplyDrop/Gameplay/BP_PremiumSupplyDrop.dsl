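; Premium supply drop Blueprint that extends its parent drop pod's states with a rift opening, player markers and orange screen and glass tints.
; Pre-fall broadcasts a gameplay event, waits a tick to spread cost, then spawns cosmetic rift VFX and audio with hotfixable duration and marks the pod server-side for every player.
; Falling shuts down the rift effects, landed-opened deactivates the special-actor component, and EndPlay deactivates leftover VFX.
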(event EventSetStatePreFall
  (|CallParentFunction)
  (GameplayEventRouter|BroadcastEvent (GameplayEventRouter|GetGlobalEventRouter) (Utilities|Struct|MakePremiumSupplyDropEventData self) 0)
  (Utilities|FlowControl|DelayUntilNextTick) ; Split perf cost across frames
  (CallFunction|SpawnRiftFX)
  (CallFunction|MarkPodforPlayers))

(event Custom|SpawnRiftFX
  (bind _returnvalue (Transformation|GetActorLocation))
  (bind _riftburstsound (Variables|FX|GetRiftBurstSound))
  (Networking|CanExecuteCosmeticEvents
    (:True
      (bind _returnvalue_2 (Niagara|SpawnSystematLocation (Variables|FX|GetOpeningRiftVFX) (Transformation|GetActorLocation) (Transformation|GetActorRotation)))
      (bind _output_get (Variables|Default|SetOpeningRiftVFXSystem _returnvalue_2))
      (Utilities|IsValid _output_get
        (:Is%20Valid
          (Niagara|SetNiagaraVariable(Float) _output_get "EffectDuration" (ScalableFloat|GetValueatLevel (Variables|Hotfix|GetHotix_RiftDuration))))
        (:Is%20Not%20Valid
          _))
      (Audio|PlaySoundatLocation _riftburstsound _returnvalue)
      (SoundIndicator|StartSoundIndicatorLooping (Variables|Default|GetRiftFortSoundIndicator) _returnvalue _riftburstsound 1.0 0)
      (Audio|Components|Audio|FadeIn (Variables|Default|GetRiftSoundLooping) 1.0 1.0 0.0 "Linear" :!class /Script/Engine.AudioComponent))))

(event Custom|MarkPodForPlayers
  ; Ping
  (Networking|SwitchHasAuthority
    (:Authority
      (for _array_element (Fortnite|GetAllFortPlayerControllers true)
        (Utilities|IsValid _array_element
          (:Is%20Valid
            (MarkerComponent|MarkActoronServer (Actor|GetComponentbyClass _array_element "/Script/FortniteGame.AthenaMarkerComponent") self false true)))))))

(event EventSetStateFalling
  (|CallParentFunction)
  (bind _openingriftvfxsystem (Variables|Default|GetOpeningRiftVFXSystem)
    (:then
      (Components|Activation|Deactivate _openingriftvfxsystem)
      (Audio|Components|Audio|FadeOut (Variables|Default|GetRiftSoundLooping) 2.5 0.0 "Linear" :!class /Script/Engine.AudioComponent)
      (SoundIndicator|StopSoundIndicatorLooping (Variables|Default|GetRiftFortSoundIndicator)))
    (:else)))

(event EventSetStateLandedOpened
  (|CallParentFunction)
  (Components|Activation|Deactivate (Variables|Default|GetAthenaSpecialActor)))

(event EventEndPlay
  (params (EndPlayReason :type EEndPlayReason%20Enum :default "Destroyed"))
  (|CallParentFunction EndPlayReason)
  (bind _openingriftvfxsystem (Variables|Default|GetOpeningRiftVFXSystem)
    (:then
      (Components|Activation|Deactivate _openingriftvfxsystem))
    (:else)))

(fn SetupDropPodMaterials
  (bind _staticmeshcomponent (Variables|Debug|GetStaticMeshComponent))
  (|CallParentFunction)
  ; To Change the Screen and Glass Color
  (bind _returnvalue (Rendering|Material|CreateDynamicMaterialInstance _staticmeshcomponent 0 (Rendering|Material|GetMaterial _staticmeshcomponent 0) "None" :!class /Script/Engine.PrimitiveComponent))
  (Rendering|Material|SetVectorParameterValue _returnvalue "DefaultScreenTint" "(R=1.000000,G=0.250923,B=0.000000,A=1.000000)" :!class /Script/Engine.MaterialInstanceDynamic)
  (bind _returnvalue_2 (Rendering|Material|CreateDynamicMaterialInstance _staticmeshcomponent 1 (Rendering|Material|GetMaterial _staticmeshcomponent 1) "None" :!class /Script/Engine.PrimitiveComponent))
  (Rendering|Material|SetVectorParameterValue _returnvalue_2 "Tint_Color" "(R=1.000000,G=0.250923,B=0.000000,A=1.000000)" :!class /Script/Engine.MaterialInstanceDynamic))

(fn ConstructionScript
  (|CallParentFunction))
