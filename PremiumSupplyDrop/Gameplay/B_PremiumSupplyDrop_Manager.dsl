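; Server-driven supply drop manager that spawns a premium drop pod ringed by mini supply drops.
; The playlist's team size, clamped to one through four, sets mini drop count, their loot tier pool and premium pod loot.
; Drop locations are filtered against the next safe zone and checked for distance from storm center.
; A determinism config location forces a drop at BeginPlay and stops phase-driven spawning; EQS debug helpers are included.
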(event EventBeginPlay
  ; Scale Mini Drop Pods based on team size.
  (|CallParentFunction)
  (if (Networking|IsServer)
    (bind (_gamestate _playlist) (Class|FortAsyncActionCurrentPlaylistReady|PlaylistDataReadyAsync)
      (:OnReady
        (Variables|Default|SetPlaylistTeamSize (Math|Integer|Clamp(Integer) (Variables|/Script/FortniteGame.FortPlaylist|GetMaxTeamSize _playlist) 1 4))
        (Variables|Default|SetMini%20Drop%20Supply%20Count%20Level (Math|Conversions|ToFloat(Integer) (Math|Integer|Clamp(Integer) (Variables|/Script/FortniteGame.FortPlaylist|GetMaxTeamSize _playlist) 1 4)))
        (CallFunction|SetUpLTGList (Math|Integer|Clamp(Integer) (Variables|/Script/FortniteGame.FortPlaylist|GetMaxTeamSize _playlist) 1 4))
        ;[0.15 0.09121 0.01429 0.5] If determinism is enabled, spawn supply drop at specified location duing begin play
        (if (not (Math|Vector|VectorIsZero (Class|FortDeterminismConfig|GetDropPodMMODeterministicSpawnLocation)))
          (CallFunction|SpawnDeterministicSupplyDrop))))))

(event Custom|DebugPremiumSupplyDropEQS
  ;[0 0 0 0.5] Cheat ke* DebugPremiumSupplyDropEQS
  (bind _outvalidlocations (Query|GetValidQueryLocationsforCurrentStormCircle))
  (Variables|Query|SetPotentialDropLocations _outvalidlocations)
  (Spawning|SpawnSupplyDropActor (Spawning|GetSupplyDropTargetLocation)))

(event Custom|SetLootForPremiumPod
  (params (Pod :type Actor%20Object%20Reference))
  ;[0.15 0.06406 0.142 0.5] Update the premium drop pod's loot based on the playlist.
  (bind _asbp_premium_supply_drop (Utilities|Casting|CastToBP_PremiumSupplyDrop Pod))
  (Variables|/DropPodParentGameplay/Gameplay/BP_SupplyDropPod_Parent.BP_SupplyDropPod_Parent_C|SetLoot%20Tier%20Name (Utilities|Select :Index (- (Variables|Default|GetPlaylistTeamSize) 1) :Option%200 "LTG_Drop_Premium_Solo" :Option%201 "LTG_Drop_Premium_Duo" :Option%202 "LTG_Drop_Premium_Trio" :Option%203 "LTG_Drop_Premium_Squad") _asbp_premium_supply_drop))

(fn GetSupplyDropTargetLocation
  (bind _droplocation (|CallParentFunction))
  ;[0.15 0.1088 0.01543 0.5] Only return a location that is within a certain distance from the center to avoid drop pods outside of the next storm radius. If a valid location isn't found within a certain amount of times just return parent result
  (bind _returnvalue (SafeZone|GetCurrentSafeZonePhase))
  (bind (_outlocation _returnvalue_2) (SafeZone|GetSafeZoneLocation _returnvalue))
  (for _ (range (+ (ScalableFloat|GetValueasInteger (Variables|Hotfix|GetHotfix_LocationSelectionMaxAttempts)) 1))
    (if (and _returnvalue_2 (or (<= (Math|Vector|Distance2DSquared(Vector) _droplocation _outlocation) (Math|Float|Square (ScalableFloat|GetValueatLevel (Variables|Hotfix|GetHotfix_MaxDistanceFromStormCenter)))) (<= (ScalableFloat|GetValueatLevel (Variables|Hotfix|GetHotfix_MaxDistanceFromStormCenter)) 0)))
      (return _droplocation)))
  (return _droplocation))

(fn GetUniqueLTG
  (if (not (<= (Utilities|Array|Length (Variables|Default|GetAvailableLTGs_MiniSupplyDrop)) 0))
    (return (Utilities|Array|Get(aref) (Variables|Default|GetAvailableLTGs_MiniSupplyDrop)))))

(fn IsQueryLocationValid
  (params (QueryLocation :type Vector :default "0, 0, 0"))
  (bind _returnvalue (SafeZone|GetCurrentPhaseIndex (SafeZone|GetSafeZoneIndicator)))
  (bind _returnvalue_2 (+ _returnvalue 1))
  (bind (_outsafezonephase _returnvalue_3) (SafeZone|GetPhaseInfo (SafeZone|GetSafeZoneIndicator) _returnvalue_2))
  (bind (_center _radius) (Utilities|Struct|BreakFortSafeZonePhaseInfo _outsafezonephase))
  (bind _returnvalue_4 (SafeZone|GetSafeZoneIndicator))
  ;[0.15 0.1279 0.01923 0.5] Return true if location is within next safe zone. Checking next safe zone due to long delay before actually falling
  (Utilities|IsValid _returnvalue_4
    (:Is%20Valid
      (if _returnvalue_3
        (return (< (Math|Vector|Distance2DSquared(Vector) _center (Variables|GetQueryLocation)) (Math|Float|Square _radius)))
        (else
          _)))
    (:Is%20Not%20Valid
      _))
  (bind _bisvalid (|CallParentFunction (Variables|GetQueryLocation)))
  (return _bisvalid))

(fn OnDropLocationQueryFinished
  (params (QueryInstance :type EQS%20Query%20Instance%20Object%20Reference) (QueryStatus :type EEnvQueryStatus%20Enum :default "Processing"))
  ;[0.15 0.1434 0.05672 0.5] Help debug found EQS locations to help check for edge cases
  (|CallParentFunction QueryInstance QueryStatus)
  (if (Variables|Default|GetbDebugEQS)
    (for _array_element (Variables|Query|GetPotentialDropLocations)
      (Rendering|Debug|DrawDebugSphere _array_element 1156.0 12 "(R=0.149886,G=1.000000,B=0.000000,A=1.000000)" 4.8e+02))))

(fn OnRep_SupplyDropSpawned
  (Default|CallOnDropPodsSpawned))

(fn SetUpLTGList
  (params (TeamSize :type Integer))
  (bind _availableltgs_minisupplydrop (Variables|Default|GetAvailableLTGs_MiniSupplyDrop))
  (Utilities|Array|Clear (Variables|Default|GetAvailableLTGs_MiniSupplyDrop))
  (switch int TeamSize
    (:1
      (Utilities|Array|Add _availableltgs_minisupplydrop "Loot_AthenaTreasure")
      (Utilities|Array|Add _availableltgs_minisupplydrop "Loot_AthenaTreasure")
      (Utilities|Array|Add _availableltgs_minisupplydrop "LTG_Chest_Special"))
    (:2
      (Utilities|Array|Add _availableltgs_minisupplydrop "Loot_AthenaTreasure")
      (Utilities|Array|Add _availableltgs_minisupplydrop "Loot_ApolloTreasure_Rare")
      (Utilities|Array|Add _availableltgs_minisupplydrop "LTG_Chest_Special"))
    (:3
      (Utilities|Array|Add _availableltgs_minisupplydrop "Loot_AthenaTreasure")
      (Utilities|Array|Add _availableltgs_minisupplydrop "Loot_AthenaTreasure")
      (Utilities|Array|Add _availableltgs_minisupplydrop "Loot_ApolloTreasure_Rare")
      (Utilities|Array|Add _availableltgs_minisupplydrop "LTG_Chest_Special"))
    (:4
      (Utilities|Array|Add _availableltgs_minisupplydrop "Loot_AthenaTreasure")
      (Utilities|Array|Add _availableltgs_minisupplydrop "Loot_ApolloTreasure_Rare")
      (Utilities|Array|Add _availableltgs_minisupplydrop "Loot_ApolloTreasure_Rare")
      (Utilities|Array|Add _availableltgs_minisupplydrop "LTG_Chest_Special"))))

(fn SpawnDeterministicSupplyDrop
  ;[0.15 0.1217 0.03205 0.5] Spawn a Drop Pod using the set deterministic location, then unbind from game phase changes to disable normal spawning.
  (Spawning|SpawnSupplyDropActor (Class|FortDeterminismConfig|GetDropPodMMODeterministicSpawnLocation))
  (bind _gamephasestepupdatedasynctask (Variables|Default|GetGamePhaseStepUpdatedAsyncTask)
    (:then
      (GameplayEventRouter|StopListeningtoStatefulEvent _gamephasestepupdatedasynctask))
    (:else
      _))
  (bind _creativestormshieldgamephasestepupdatedasynctask (Variables|Default|GetCreativeStormShieldGamePhaseStepUpdatedAsyncTask)
    (:then
      (GameplayEventRouter|StopListeningtoStatefulEvent _creativestormshieldgamephasestepupdatedasynctask))
    (:else)))

(fn SpawnSupplyDropActor
  (params (DropLocation :type Vector :default "0, 0, 0"))
  (bind _mini_drop_supply_count_level (Variables|Default|GetMini%20Drop%20Supply%20Count%20Level))
  (bind _hotfix_minisupplydropcount (Variables|Hotfix|GetHotfix_MiniSupplyDropCount))
  (bind _returnvalue (ScalableFloat|GetValueasInteger _hotfix_minisupplydropcount _mini_drop_supply_count_level))
  ;[0.06507 0.15 0.03965 0.5] Spawn mini supply drops in circle around drop location, then use parent function to spawn premium supply drop in center
  (Utilities|FlowControl|Sequence
    (:then_0
      (for _index (range 1 (+ _returnvalue 1))
        (Game|SpawnActorfromClass (Variables|Default|GetMiniSupplyDropClass) (Math|Transform|MakeTransform (+ (+ DropLocation (Variables|SpawnRules|GetSpawnPositionOffset)) (Math|Vector|RotateVectorAroundAxis (Math|Vector|MakeVector (ScalableFloat|GetValueatLevel (Variables|Hotfix|GetHotfix_MiniSupplyDropRadius))) (* _index (/ 3.6e+02 (Math|Conversions|ToFloat(Integer) _returnvalue))) "0.000000,0.000000,1.000000"))) "Undefined" "MultiplyWithRoot" 0 0)))
    (:then_1
      (bind _returnvalue_2 (|CallParentFunction DropLocation))
      (CallFunction|SetLootforPremiumPod _returnvalue_2) ; The premium pod is spawned by the parent call.\nUpdate the loot tier group.
      (Variables|Default|SetSupplyDropSpawned true)
      (return _returnvalue_2))))

(fn ConstructionScript
  (|CallParentFunction))
