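# Environment query that picks candidate locations for premium supply drops around the querier.
# A very large, widely spaced simple grid generates points and projects them vertically onto geometry with line traces.
# Two box overlap tests on separate collision channels and a flat-surface validity test, which filters and scores, evaluate each point for room and footing.
# A random test adds seeded score variation so the chosen spot varies.
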EnvQuery /PremiumSupplyDrop/Gameplay/EQS_PremiumSupplyDrop_Locations
options [{
  generator {
   name=FortEnvQueryGenerator_SimpleGrid_0
   class=FortEnvQueryGenerator_SimpleGrid
   properties {
    bAutoSortTests=true
    bCanRunAsync=false
    envQueryResultNormalizationOption=Default
    generateAround=&EnvQueryContext_Querier
    gridSize=75000
    navDataOverrideContext=null
    optionName=""
    overridenAgentNameForNavmesh=None
    randomDisplacementRatio=0
    spaceBetween=2048
    projectionData {
     bCanDisableTrace=true
     bCanProjectDown=true
     bCanTraceOnGeometry=true
     bCanTraceOnNavMesh=true
     bOnlyBlockingHits=true
     bTraceComplex=false
     extentX=0
     extentY=10
     extentZ=10
     navigationFilter=null
     postProjectionVerticalOffset=0
     projectDown=30000
     projectUp=10000
     serializedChannel=ECC_GameTraceChannel6
     traceChannel=TraceTypeQuery7
     traceMode=GeometryByChannel
     traceProfileName=None
     traceShape=Line
     versionNum=1
    }
   }
  }
  tests [
   EnvQueryTest_Overlap_3 {
    class=EnvQueryTest_Overlap
    properties {overlapData {extentX=1156, extentY=1156, extentZ=1028, overlapChannel=ECC_GameTraceChannel15, shapeOffset=<0 0 0>}}
   }
   FortQueryTest_ValidSurface_0 {
    class=FortQueryTest_ValidSurface
    properties {
     flatSurfaceToleranceZ=200
     radius=1028
     testPurpose=FilterAndScore
     traceCollisionChannel=ECC_GameTraceChannel6
     traceOffsetDown=500
     traceOffsetUp=500
     surfaceMaterials []
    }
   }
   EnvQueryTest_Overlap_2 {
    class=EnvQueryTest_Overlap
    properties {overlapData {extentX=1028, extentY=1156, extentZ=1156, overlapChannel=ECC_GameTraceChannel1, shapeOffset=<0 0 0>}}
   }
   FortQueryTest_Random_0 {class=FortQueryTest_Random, properties {bUseRandomSeedForAI=false, bUseRandomSeedForOthers=true, floatValueMax=0, scoringFactor=1}}
  ]
 }]
